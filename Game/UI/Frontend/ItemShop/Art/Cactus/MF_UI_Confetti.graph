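# Material function for Item Shop UI that draws procedural animated confetti, outputting a mask plus that mask multiplied by the scroll offset as RandomNoise.
# Rotated, tiled UVs form cells, each hashed with a seed into a random value that varies scroll speed, scale and sine-driven wobble.
# A custom-node parallelogram signed distance field draws each piece without textures, and a smoothstep with adjustable range sets its edge.
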MaterialFunction /Game/UI/Frontend/ItemShop/Art/Cactus/MF_UI_Confetti
graph {nodes [
  n1 = MaterialFunctionCall(Float2=FunctionInput(Preview=Constant2Vector() {g=1, r=0.5}) {bUsePreviewValueAsDefault=true, inputName="SmoothStepRanges(V2)", inputType=FunctionInput_Vector2, sortPriority=9}) {materialFunction=&BreakOutFloat2Components}
  n2 = TextureCoordinate()
  n3 = Constant() {r=-0.125}
  n4 = Multiply(
   A=MaterialFunctionCall(
    UVs=FunctionInput(Preview=@n2) {bUsePreviewValueAsDefault=true, inputName="UV(V2)", inputType=FunctionInput_Vector2}
    "Rotation Angle (0-1)"=FunctionInput(Preview=@n3) {bUsePreviewValueAsDefault=true, inputName=Global-Angle, inputType=FunctionInput_Scalar, sortPriority=1}
   ) {materialFunction=&CustomRotator}.Rotated Values
   B=FunctionInput(Preview=Constant2Vector() {g=1, r=10}) {bUsePreviewValueAsDefault=true, inputName="Pattern-Tiles(V2)", inputType=FunctionInput_Vector2, sortPriority=3}
  )
  n5 = Constant() {r=-0.05}
  n6 = Multiply(
   A=NamedRerouteUsage()
   B=Multiply(
    A=Time()
    B=FunctionInput(Preview=@n5) {bUsePreviewValueAsDefault=true, inputName=Confetti-Speed-TimeFactor, inputType=FunctionInput_Scalar, sortPriority=5}
   )
  )
  n7 = AppendVector(A=Constant(), B=ConstantBiasScale(None=Frac(None=@n6)) {bias=-0.5, scale=2})
  n8 = MaterialFunctionCall(UVs=Add(A=Frac(None=@n4), B=@n7), Texture Scale=NamedRerouteUsage()) {materialFunction=&ScaleUVsByCenter}
  n9 = FunctionInput(Preview=Constant() {r=0.0075}) {bUsePreviewValueAsDefault=true, inputName=Confetti-Local-Rotation-Range, inputType=FunctionInput_Scalar, sortPriority=7}
  n10 = Multiply(
   A=Time()
   B=FunctionInput(Preview=Constant() {r=0.25}) {bUsePreviewValueAsDefault=true, inputName=Confetti-Rotation-TimeFactor, inputType=FunctionInput_Scalar, sortPriority=6}
  )
  n11 = NamedRerouteUsage()
  n12 = LinearInterpolate(A=Multiply(A=@n9) {constB=-1}, B=@n9, Alpha=Sine(None=Multiply(A=@n10, B=@n11))) {constA=0.001, constB=0.001}
  n13 = Constant2Vector() {g=10, r=1}
  MaterialExpressionAdd_8 = Add(
   A=Multiply(
    A=Rotator(Coordinate=@n8.UVs, Time=@n12) {speed=1}
    B=FunctionInput(Preview=@n13) {bUsePreviewValueAsDefault=true, inputName="Confetti-Local-Tiles(V2)", inputType=FunctionInput_Vector2, sortPriority=8}
   )
   B=ScalarParameter() {defaultValue=-0.5, parameterName=p}
  )
  MaterialExpressionScalarParameter_21 = ScalarParameter() {defaultValue=0.4, parameterName=wi}
  MaterialExpressionScalarParameter_22 = ScalarParameter() {parameterName=sk}
  MaterialExpressionScalarParameter_23 = ScalarParameter() {defaultValue=0.75, parameterName=he}
  n14 = Custom!(
   p=@MaterialExpressionAdd_8
   wi=@MaterialExpressionScalarParameter_21
   sk=@MaterialExpressionScalarParameter_22
   he=@MaterialExpressionScalarParameter_23
  ) {
   code="float2  e  = float2(sk,he);\r\nfloat e2 = sk*sk + he*he;\r\np = (p.y<0.0)?-p:p;\r\nfloat2  w = p - e; w.x -= clamp(w.x,-wi,wi);\r\nfloat2  d = float2(dot(w,w), -w.y);\r\nfloat s = p.x*e.y - p.y*e.x;\r\np = (s<0.0)?-p:p;\r\nfloat2  v = p - float2(wi,0); v -= e*clamp(dot(v,e)/e2,-1.0,1.0);\r\nd = min( d, float2(dot(v,v), wi*he-abs(s)));\r\nreturn sqrt(d.x)*sign(-d.y);"
   description=Parallelogram
   outputType=CMOT_Float2
   inputs [
    {inputName=p, input {expression=&self:MaterialExpressionAdd_8}}
    {inputName=wi, input {expression=&self:MaterialExpressionScalarParameter_21}}
    {inputName=sk, input {expression=&self:MaterialExpressionScalarParameter_22}}
    {inputName=he, input {expression=&self:MaterialExpressionScalarParameter_23}}
   ]
  }
  n15 = OneMinus(None=SmoothStep(Min=Subtract(A=@n1.R, B=@n1.R) {constB=0.25}, Max=@n1.R, Value=@n14) {constMax=0.6, constMin=0.5, constValue=0.5})
  FunctionOutput(None=ComponentMask(None=@n15) {r=true}) {outputName=Confetti-Mask}
  FunctionOutput(None=Multiply(A=@n15, B=@n7)) {outputName=RandomNoise, sortPriority=1}
  MaterialExpressionAdd_6 = Add(
   A=FunctionInput(Preview=Constant() {r=1234}) {bUsePreviewValueAsDefault=true, inputName=RandomSeed, inputType=FunctionInput_Scalar, sortPriority=4}
   B=Add(A=Floor(None=@n4)) {constB=0.01}
  )
  n16 = Custom!(x=@MaterialExpressionAdd_6) {
   code="uint2 q = 1103515245U * ( (asuint(x)>>1U) ^ (asuint((x.yx))   ) );\r\nuint  n = 1103515245U * ( (q.x  ) ^ (q.y>>3U) );\r\nreturn float(n) * (1.0/float(0xffffffffU));"
   desc=Random
   description=IQHash3
   outputType=CMOT_Float1
   inputs [{inputName=x, input {expression=&self:MaterialExpressionAdd_6}}]
  }
  NamedRerouteDeclaration(None=@n16) {name=Random, nodeColor=<rgba 0 0 0 1>}
 ]}
